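# Material function that generates a rectangular mask from texture coordinates, in the GroovyReader backpack's material folder.
# It subtracts the left, right, up and down position inputs from the UVs and applies SmoothStep to each edge, with SquareSoftness scaled down to set the falloff.
# Right and down edges are inverted, then the horizontal and vertical bands multiply into SquareMask.
# A static switch optionally rotates the UVs first, feeding RotationAmount into a Rotator's Time input.
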MaterialFunction /Game/Accessories/FORT_Backpacks/Backpack_F_MED_GroovyReader_Play/Materials/MaterialFunctions/MF_SquareMaskGenerator
graph {nodes [
  n1 = FunctionInput() {bUsePreviewValueAsDefault=true, inputName=X Pos Right, inputType=FunctionInput_Scalar, previewValue=<0.5 0 0 1>}
  n2 = FunctionInput() {inputName=Texture Coordinate, inputType=FunctionInput_Vector2}
  n3 = StaticSwitch(
   True=Rotator(
    Coordinate=@n2
    Time=FunctionInput() {bUsePreviewValueAsDefault=true, inputName=RotationAmount, inputType=FunctionInput_Scalar, sortPriority=7}
   )
   False=@n2
   Value=FunctionInput(Preview=StaticBool()) {bUsePreviewValueAsDefault=true, inputName=RotateSquare, inputType=FunctionInput_StaticBool, sortPriority=6}
  )
  n4 = ComponentMask(None=Add(A=Multiply(A=@n1) {constB=-1}, B=@n3)) {r=true}
  n5 = FunctionInput() {bUsePreviewValueAsDefault=true, inputName=SquareSoftness, inputType=FunctionInput_Scalar, sortPriority=5, previewValue=<1 0 0 1>}
  n6 = Multiply(A=@n5) {constB=0.001}
  n7 = LinearInterpolate(Alpha=Saturate(None=SmoothStep(Max=@n6, Value=@n4) {constMax=0.001})) {constA=1, constB=0}
  n8 = FunctionInput() {bUsePreviewValueAsDefault=true, inputName=X Pos Left, inputType=FunctionInput_Scalar, sortPriority=2, previewValue=<0.2 0 0 1>}
  n9 = Add(A=@n3, B=Multiply(A=@n8) {constB=-1})
  n10 = Saturate(None=SmoothStep(Max=@n6, Value=ComponentMask(None=@n9) {r=true}) {constMax=0.001})
  n11 = FunctionInput() {bUsePreviewValueAsDefault=true, inputName=Y Pos Down, inputType=FunctionInput_Scalar, sortPriority=4, previewValue=<0.9 0 0 1>}
  n12 = ComponentMask(None=Add(A=@n3, B=Multiply(A=@n11) {constB=-1})) {g=true}
  n13 = LinearInterpolate(Alpha=Saturate(None=SmoothStep(Max=@n6, Value=@n12) {constMax=0.001})) {constA=1, constB=0}
  n14 = FunctionInput() {bUsePreviewValueAsDefault=true, inputName=Y Pos Up, inputType=FunctionInput_Scalar, sortPriority=3, previewValue=<0.2 0 0 1>}
  n15 = Add(A=@n3, B=Multiply(A=@n14) {constB=-1})
  n16 = Saturate(None=SmoothStep(Max=@n6, Value=ComponentMask(None=@n15) {g=true}) {constMax=0.001})
  FunctionOutput(None=Multiply(A=LinearInterpolate(B=@n7, Alpha=@n10) {constB=0}, B=LinearInterpolate(B=@n13, Alpha=@n16) {constB=0})) {outputName=SquareMask}
 ]}
